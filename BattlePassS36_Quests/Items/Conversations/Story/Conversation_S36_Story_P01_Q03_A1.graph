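# Conversation database for a Season 36 story quest (P01 Q03) NPC dialogue, tagged Morgan, who thanks the player for stabilizing their shadow proxies.
# A single entry point leads to a "Greet" choice with a main-quest icon, gated on that quest being active.
# It then runs linearly through two "Continue" speech nodes about Hope and the speaker's 'potential', and ends with no branching.
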ConversationDatabase /BattlePassS36_Quests/Items/Conversations/Story/Conversation_S36_Story_P01_Q03_A1
entry_points [{tag=Conversation.Entry.Season36.Story.P01.Q03.01, destinations [#1]}]
nodes [
 ConversationEntryPointNode_1 {
  id=#1
  class=ConversationEntryPointNode
  outputs [#2]
  properties {entryTag {tagName=Conversation.Entry.Season36.Story.P01.Q03.01}}
 }
 Speech_C_10 {
  id=#2
  class=Speech_C
  outputs [#3]
  properties {
   softTaskIcon=&T_UI_NPC_MainQuest
   generalConfig {defaultMessage="You're the one who stabilized my shadow proxies, right? Thanks."}
  }
  sub_nodes [
   ConversationChoiceNode_2 {
    class=ConversationChoiceNode
    properties {defaultChoiceDisplayText=Greet, choiceTags {gameplayTags [{tagName=Athena.Quests.S36.Conversations.Story.P01.Q3Morgan}]}}
   }
   FortConversationRequirement_ControllerRequirement_1 {
    class=FortConversationRequirement_ControllerRequirement
    properties {
     nodeName=Require: Is Quest Active
     requirement {
      class=FortControllerRequirement_IsQuestActive
      properties {questDefinition=/BattlePassS36_Quests/Items/QuestItems/Story/P01/Quest_S36_Story_P01_Q03}
     }
    }
   }
  ]
 }
 Speech_C_0 {
  id=#3
  class=Speech_C
  outputs [#4]
  properties {generalConfig {defaultMessage=Hope said she knew someone who could help. I guess that's you.}, voiceOverTag {tagName=None}}
  sub_nodes [ConversationChoiceNode_3 {class=ConversationChoiceNode, properties {defaultChoiceDisplayText=Continue}}]
 }
 Speech_C_2 {
  id=#4
  class=Speech_C
  outputs []
  properties {generalConfig {defaultMessage=She keeps saying I have 'potential'. Whatever that means.}, voiceOverTag {tagName=None}}
  sub_nodes [ConversationChoiceNode_4 {class=ConversationChoiceNode, properties {defaultChoiceDisplayText=Continue}}]
 }
]
